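# Niagara idle effect for the FutureRabbit backpack cosmetic, adding glows for its screen, face, ears and sides.
# ScreenFaceEars bursts particles onto sockets chosen by partition and spawn group, drawing two sprite layers, a light and a mesh with waveform color flicker.
# Sides places two oriented mesh particles at their own sockets.
# System variant index selects color sets, and a time-of-day scalar scales alpha.
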NiagaraSystem /BRCosmetics/Accessories/FORT_Backpacks/Backpack_F_MED_FutureRabbit/FX/NS_Backpack_FutureRabbit_Idle
dependencies {
 usedDataInterfaces [
  &NiagaraDataInterfaceArrayFloat
  &NiagaraDataInterfaceArrayInt32
  &NiagaraDataInterfaceSocketReader
 ]
 usedDynamicInputs [&SelectFloatFromArray, &FNTimeOfDayScalar]
 usedModules [
  &FN_VariantManager
  &SystemState
  &self:SystemSpawnScript.SystemScriptSource.SystemScriptGraph.NiagaraNodeAssignment_0.SetVariables_#1
  &self:SystemSpawnScript.SystemScriptSource.SystemScriptGraph.NiagaraNodeAssignment_1.SetVariables_#2
 ]
}
emitters {
 ScreenFaceEars {
  emitterUpdateScript [
   &EmitterState()
   &SpawnBurst_Instantaneous(Spawn Count=6, Use Spawn Probability=false, UseLoopCountLimit=true)
   &SpawnBurst_Instantaneous(Spawn Group=1, Use Spawn Probability=false, UseLoopCountLimit=true) {module=SpawnBurst_Instantaneous001}
   &SpawnBurst_Instantaneous(Spawn Time=8.0667, Use Spawn Probability=false, UseLoopCountLimit=false) {module=SpawnBurst_Instantaneous002}
  ]
  particleSpawnScript [
   /SetVariables(Particles.VisibilityTag=$Emitter.SpawnGroup) {isSetParameters=true}
   &PartitionParticles(Number Of Partitions=8)
   &InitializeParticle(
    Color=&MergeHelper_ColorSwitch(
     0=&VariantColorSelector(
      Variant Amount=4
      Original Color=<rgba 0 0.5167 1 0.25>
      Variant 1 Color=<rgba 1 0 0.1565 0.25>
      Variant 2 Color=<rgba 0 1 0.4522 0.25>
      Variant 3 Color=<rgba 1 0.5739 0 0.25>
      Variant 4 Color=<rgba 1 0.3 1 0.25>
      Variant 5 Color=<rgba 0 0.6667 1 0.25>
     )
     1=&VariantColorSelector(
      Variant Amount=4
      Original Color=<rgba 0 0.5167 1 0.15>
      Variant 1 Color=<rgba 1 0 0.1565 0.15>
      Variant 2 Color=<rgba 0 1 0.4522 0.15>
      Variant 3 Color=<rgba 1 0.5739 0 0.15>
      Variant 4 Color=<rgba 0 0.3333 1 0.15>
      Variant 5 Color=<rgba 0 0.6667 1 0.15>
     )
     2=?
     3=?
     Index=$Particles.PartitionParticles.Partition
     Num=::NewEnumerator0
    )
    Lifetime=2.5
    Position Mode=::NewEnumerator0
    Sprite Rotation Angle=&MergeHelper_FloatSwitch(0=0, 1=0, 2=-50, 3=50, 4=50, 5=-50, Index=$Particles.PartitionParticles.Partition, Num=::NewEnumerator4)
    Sprite Rotation Mode=::NewEnumerator0
    Sprite Size=&MergeHelper_Vector2Switch(0=<35 45>, 1=<12.5 12.5>, 2=<0 0>, 3=<0 0>, 4=<0 0>, 5=<0 0>, Index=$Particles.PartitionParticles.Partition, Num=::NewEnumerator4)
    Sprite Size Mode=::NewEnumerator4
    Write Material Random=true
   )
   &DynamicMaterialParameters(
    Index 0 Param 1=1
    Index 0 Param 2=0.7
    Module.Index0Float0Write=true
    Module.Index0Float1Write=true
    Module.Index0Float2Write=true
    Module.Index0Float3Write=true
   )
   /SetVariables(Particles.BaseColor:LinearColor=&MergeHelper_ColorSwitch(
     0=<rgba 0.6106 1.5424 4 1>
     1=&MergeHelper_ColorSwitch(
      0=<rgba 0.6106 1.5424 4 1>
      1=<rgba 4 0.6106 1.1755 1>
      2=<rgba 0.6106 4 2.1358 1>
      3=<rgba 4 2.5313 0.6106 1>
      4=<rgba 0.6106 1.7404 4 1>
      5=<rgba 0.6106 2.8702 4 1>
      Index=&MakeIntFromFloat(Float=$System.VariantIndex)
      Num=::NewEnumerator4
     )
     2=&MergeHelper_ColorSwitch(
      0=<rgba 0.6106 1.5424 4 1>
      1=<rgba 0.6106 1.5424 4 1>
      2=<rgba 0.6106 4 2.1358 1>
      3=<rgba 4 2.5313 0.6106 1>
      4=<rgba 0.6106 1.7404 4 1>
      5=<rgba 0.6106 2.8702 4 1>
      Index=&MakeIntFromFloat(Float=$System.VariantIndex)
      Num=::NewEnumerator4
     )
     3=&MergeHelper_ColorSwitch(
      0=<rgba 0.6106 1.5424 4 1>
      1=<rgba 0.6106 1.5424 4 1>
      2=<rgba 0.6106 4 2.1358 1>
      3=<rgba 4 2.5313 0.6106 1>
      4=<rgba 0.6106 1.7404 4 1>
      5=<rgba 0.6106 2.8702 4 1>
      Index=&MakeIntFromFloat(Float=$System.VariantIndex)
      Num=::NewEnumerator4
     )
     4=&MergeHelper_ColorSwitch(
      0=<rgba 0.6106 1.5424 4 1>
      1=<rgba 0.6106 1.5424 4 1>
      2=<rgba 0.6106 4 2.1358 1>
      3=<rgba 4 2.5313 0.6106 1>
      4=<rgba 0.6106 1.7404 4 1>
      5=<rgba 0.6106 2.8702 4 1>
      Index=&MakeIntFromFloat(Float=$System.VariantIndex)
      Num=::NewEnumerator4
     )
     5=&MergeHelper_ColorSwitch(
      0=<rgba 0.6106 1.5424 4 1>
      1=<rgba 0.6106 1.5424 4 1>
      2=<rgba 0.6106 4 2.1358 1>
      3=<rgba 4 2.5313 0.6106 1>
      4=<rgba 0.6106 1.7404 4 1>
      5=<rgba 0.6106 2.8702 4 1>
      Index=&MakeIntFromFloat(Float=$System.VariantIndex)
      Num=::NewEnumerator4
     )
     Index=$Particles.PartitionParticles.Partition
     Num=::NewEnumerator4
    )) {isSetParameters=true}
  ]
  particleUpdateScript [
   &ParticleState(Kill Particles When Lifetime Has Elapsed=false, Let Infinitely Lived Particles Die When Emitter Deactivates=true)
   &SocketLocation(
    SamplePosOffset=false
    Socket Index=&MergeHelper_IntSwitch(
     0=&SelectIntFromArray(Direct Array Index=$Particles.PartitionParticles.Partition, Int Selection Array=DataInterface() {propertyValues {IntData [0,1,2,3,4,5]}})
     1=8
     2=?
     3=?
     4=?
     5=?
     Index=$Particles.VisibilityTag
     Num=::NewEnumerator0
    )
    Socket Reader=$System.Socket Reader
    Socket Sampling Mode=::NewEnumerator3
    VelocitySpeedLimit=false
   )
   /SetVariables(
    Particles.BaseAlignment:Vector3f=<0 1 0>
    Particles.BaseDynamic:Vector4f=<0.85 0.2 1 1>
    Particles.BaseFacing:Vector3f=&MergeHelper_VectorSwitch(
     0=<0 -1 0.584>
     1=<0 -0.6 -1>
     2=<1 0 0>
     3=<-1 0 0>
     4=<-0.1 0 0>
     5=<1 0 0>
     Index=$Particles.PartitionParticles.Partition
     Num=::NewEnumerator4
    )
    Particles.BasePosition:NiagaraPosition=&AddVectorToPosition(
     Position=$Particles.Position
     Vector=&MergeHelper_VectorSwitch(
      0=<0 -0.9781 1.9147>
      1=<0 0 0>
      2=<0 0 0>
      3=<0 0 0>
      4=<0 0 0>
      5=<0 0 0>
      Index=$Particles.PartitionParticles.Partition
      Num=::NewEnumerator4
     )
    )
    Particles.BaseSize:Vector2f=&MergeHelper_Vector2Switch(0=<0 0>, 1=<25 25>, 2=<15 7>, 3=<15 7>, 4=<15 7>, 5=<15 7>, Index=$Particles.PartitionParticles.Partition, Num=::NewEnumerator4)
   ) {isSetParameters=true}
   &CameraOffset(Camera Offset Amount=2)
   &ScaleColor(
    Scale Alpha=$System.TOD Scalar
    Scale RGB=&VectorFromFloat(Value=&Lerp_Float(
      A=1
      B=&Waveform(Global Amplitude Scale=0.5, "Waveform [1]"=::NewEnumerator8, "[1] Frequency"=15, "[1] Amplitude Min/Max"=<0.2 1>, "[1] Bias"=0.7)
      Alpha=0.4
     ))
    ScaleA=true
    ScaleRGB=true
   )
   &FN_Light_Attributes(
    Color=&MultiplyLinearColor(Linear Color=$Particles.Color, Scale Factor=&MakeVector4_FromLinearColor(LinearColor=$Particles.Color))
    Create Color Attribute=true
    Create Exponent Attribute=true
    Create Radius Attribute=true
    Enable Particle Light Enable=true
    Exponent=15
    Light Enabled=true
    Radius=35
    Scale Color=10
   )
   /SetVariables(Particles.BaseColor:LinearColor=&MultiplyLinearColor(
     Linear Color=$Particles.Initial.BaseColor
     Scale Factor=&MakeVector4FromVectorAndFloat("Vector (RGB)"=<1 1 1>, "Float (Alpha)"=0.5)
    )) {isSetParameters=true, module=SetVariables_#3}
   &FN_Mesh_Attributes(
    Color=&VariantColorSelector(
     Variant Amount=4
     Original Color=<rgba 0 0.5 2 1>
     Variant 1 Color=<rgba 2 0 0.3131 1>
     Variant 2 Color=<rgba 0 1 0.45 1>
     Variant 3 Color=<rgba 1 0.5667 0 1>
     Variant 4 Color=<rgba 0.5 0.2 0.5 1>
     Variant 5 Color=<rgba 0 1.3333 2 1>
    )
    Create Camera Offset Attribute=true
    Create Color Attribute=true
    Create Orientation Attribute=true
    Create Position Attribute=true
    Create Scale Attribute=true
    Module Output=::NewEnumerator1
    Orientation:Quat4f=<-0.0065 0 0 1>
    Position:NiagaraPosition=&AddVectorToPosition(Position=$Engine.Emitter.SimulationPosition, Vector=<0 -1.22 -0.1>)
    Scale=&VectorFromFloat(Value=0.927)
   )
  ]
  rendererClasses [
   &NiagaraSpriteRendererProperties
   &NiagaraLightRendererProperties
   &NiagaraMeshRendererProperties
  ]
  renderers [
   {rendererClass=&NiagaraSpriteRendererProperties}
   {rendererClass=&NiagaraSpriteRendererProperties, rendererIndex=1}
   {rendererClass=&NiagaraLightRendererProperties, rendererIndex=2}
   {rendererClass=&NiagaraMeshRendererProperties, rendererIndex=3}
  ]
 }
 Sides {
  emitterUpdateScript [
   &EmitterState()
   &SpawnBurst_Instantaneous(Spawn Count=2, Use Spawn Probability=false, UseLoopCountLimit=true)
  ]
  particleSpawnScript [
   &PartitionParticles(Number Of Partitions=4)
   &InitializeParticle(
    Color=&MergeHelper_ColorSwitch(
     0=<rgba 0.15 0.5906 1 0.2>
     1=<rgba 0.15 0.5906 1 0.2>
     2=<rgba 0.15 1 0.5325 0.2>
     3=<rgba 1 0.6316 0.15 0.2>
     4=<rgba 0.15 0.4333 1 0.2>
     5=<rgba 0.15 0.7167 1 0.2>
     Index=&MakeIntFromFloat(Float=$System.VariantIndex)
     Num=::NewEnumerator4
    )
    Lifetime=2.5
    Mesh Scale=&MergeHelper_VectorSwitch(0=<0.25 0.25 0.35>, 1=<0.25 0.25 0.35>, 2=?, 3=?, Index=$Particles.PartitionParticles.Partition, Num=::NewEnumerator0)
    Mesh Scale Mode=::NewEnumerator4
    Position Mode=::NewEnumerator0
    Write Material Random=true
   )
   &SocketLocation(
    SamplePosOffset=false
    Socket Index=&SelectIntFromArray(Direct Array Index=$Particles.PartitionParticles.Partition, Int Selection Array=DataInterface() {propertyValues {IntData [6,7,2,3]}})
    Socket Reader=$System.Socket Reader
    Socket Sampling Mode=::NewEnumerator3
    VelocitySpeedLimit=false
   )
   &InitialMeshOrientation(
    Mesh Orientation Mode=::NewEnumerator1
    Rotation=&MergeHelper_VectorSwitch(0=<0.338 0 0.35>, 1=<0.338 0 -0.35>, 2=?, 3=?, Index=$Particles.PartitionParticles.Partition, Num=::NewEnumerator0)
    Use Rotation Vector=true
   )
  ]
  particleUpdateScript [
   &ParticleState(Kill Particles When Lifetime Has Elapsed=false, Let Infinitely Lived Particles Die When Emitter Deactivates=true)
   &CameraOffset(Camera Offset Amount=2)
   &ScaleColor(Scale Alpha=$System.TOD Scalar, ScaleA=true, ScaleRGB=true)
  ]
  rendererClasses [&NiagaraMeshRendererProperties]
  renderers [{rendererClass=&NiagaraMeshRendererProperties, rendererIndex=0}]
 }
}
summary {
 systemName=NS_Backpack_FutureRabbit_Idle
 emitters [
  {
   emitterName=ScreenFaceEars
   rendererClasses [
    &NiagaraSpriteRendererProperties
    &NiagaraLightRendererProperties
    &NiagaraMeshRendererProperties
   ]
  }
  {emitterName=Sides, rendererClasses [&NiagaraMeshRendererProperties]}
 ]
 userVariables [
  User.bIsFrontEnd {type=NiagaraBool}
  User.bIsFrontEndPreview {type=NiagaraBool}
  User.Variant {type=NiagaraFloat, defaultValue=0, description=Choose Variant}
 ]
}
