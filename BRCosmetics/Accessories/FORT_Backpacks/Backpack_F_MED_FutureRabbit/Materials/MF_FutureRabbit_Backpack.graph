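# Material function for the FutureRabbit backpack's visor, adding an animated, glowing face to incoming material attributes.
# Time-based blink and staggered left and right wink pulses select frames from a distance-field flipbook to draw the face.
# A procedural dot grid, masked by paired radial gradients and crescents, adds glowing spots, combined with Fresnel tinting.
# An SDF circle defines the visor mask, and the result drives emissive and scaled base color.
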MaterialFunction /BRCosmetics/Accessories/FORT_Backpacks/Backpack_F_MED_FutureRabbit/Materials/MF_FutureRabbit_Backpack
graph {nodes [
  n1 = TextureObjectParameter() {
   group=Visor Base
   parameterName=AnimatedDF_EmissiveDistanceField
   sortPriority=40
   texture=/BRCosmetics/Accessories/FORT_Backpacks/Backpack_F_MED_FutureRabbit/Textures/T_FutureRabbit_Backpack_FlipbookFace
  }
  n2 = ScalarParameter() {defaultValue=2, group=Visor Base, parameterName=AnimatedVisor_SubUV_Frames, sortPriority=60}
  n3 = VectorParameter() {group=Visor, parameterName=Visor_UVOffset, sortPriority=40, defaultValue=<rgba -0.1392 -0.3016 0 1>}
  n4 = MaterialFunctionCall!(
   Flipbook Index=NamedRerouteUsage()
   Flipbook Texture=@n1
   Flipbook Column Count=@n2
   Flipbook Row Count=@n2
   UVs=TextureCoordinate()
   UV Offset=ComponentMask(None=@n3.RGB) {g=true, r=true}
   UV Scale=ComponentMask(None=VectorParameter() {group=Visor, parameterName=Visor_UVScale, sortPriority=40, defaultValue=<rgba 0.25 -0.25 0 1>}.RGB) {g=true, r=true}
   MirrorLeftUV
  ) {materialFunction=/Game/Global/Materials/Character_Parent_2023/FlipbookFace_Parent_2023/MaterialFunctions/Common/MF_FlipbookTexture}
  n5 = Saturate(None=Multiply(B=@n4.Color) {constA=4})
  n6 = Multiply(A=VectorParameter() {group=Visor, parameterName=Visor_MaskTintColor, defaultValue=<rgba 1 1 1 1>}.RGB, B=@n5)
  n7 = Power(Base=@n4.Color)
  Add(
   A=NamedRerouteDeclaration(None=Add(A=@n6, B=@n7)) {name=Visor_Base, nodeColor=<rgba 0 0.0161 1 1>}
   B=NamedRerouteDeclaration() {name=EyesGlowArea, nodeColor=<rgba 0.8047 1 0 1>}
  )
  ChannelMaskParameter(None=FunctionInput() {inputName=Tech Art Mask, sortPriority=1}) {group=Visor, parameterName=Visor_Channel, sortPriority=19}
  Constant()
  n8 = GetMaterialAttributes(None=FunctionInput() {inputName=Material, inputType=FunctionInput_MaterialAttributes}) {attributeGetTypes [#1, #2, #3]}
  n9 = Multiply(A=@n8.BaseColor, B=ScalarParameter() {group=Visor, parameterName=Visor_BasecolorScale, sortPriority=43})
  n10 = NamedRerouteUsage()
  FunctionOutput(None=SetMaterialAttributes!(
    MaterialAttributes=@n8.MaterialAttributes
    EmissiveColor=NamedRerouteUsage()
    BaseColor=LinearInterpolate(A=@n8.BaseColor, B=@n9, Alpha=@n10) {constB=0}
   ) {attributeSetTypes [#2, #3]})
  n11 = Constant() {r=36}
  n12 = Divide(A=Time(), B=@n11) {constB=10}
  n13 = Constant() {r=1.25}
  n14 = Floor(None=Add(A=Frac(None=@n12), B=Divide(A=@n13, B=@n11) {constB=10}))
  NamedRerouteDeclaration(None=@n14) {name=RightWink, nodeColor=<rgba 0.2422 0 1 1>}
  NamedRerouteDeclaration(None=@n8.Normal) {name=Normal_IN, nodeColor=<rgba 1 0 0.7266 1>}
  n15 = Power(
   Base=OneMinus(None=Fresnel() {baseReflectFraction=0, exponent=1})
   Exp=ScalarParameter() {defaultValue=1, group=Visor, parameterName=Visor_FresnelPower, sortPriority=43}
  )
  NamedRerouteDeclaration(None=LinearInterpolate(B=Saturate(None=@n15), Alpha=ScalarParameter() {defaultValue=0.5, group=Visor, parameterName=Visor_FresnelIntensity, sortPriority=43}) {constA=1}) {name=Fresnel, nodeColor=<rgba 1 0.8203 0 1>}
  n16 = Multiply(
   A=NamedRerouteDeclaration(None=TextureCoordinate()) {name=UV12, nodeColor=<rgba 0 0.0391 1 1>}
   B=ComponentMask(None=VectorParameter() {group=Visor, parameterName=Visor_UVScale, sortPriority=40, defaultValue=<rgba 0.25 -0.25 0 1>}.RGB) {g=true, r=true}
  )
  n17 = ComponentMask(None=VectorParameter() {group=Visor, parameterName=Visor_UVOffset, sortPriority=40, defaultValue=<rgba -0.6 -0.5 0 1>}.RGB) {g=true, r=true}
  NamedRerouteDeclaration(None=MaterialFunctionCall!(
    UV Coord=Add(A=@n16, B=@n17)
    Height=ScalarParameter() {group=Visor, parameterName=Visor_Depth, sortPriority=41}
    ReferencePlane=Constant()
    HeightRatio
    Normal=NamedRerouteUsage()
   ) {materialFunction=/Game/Global/MaterialFunctions/MF_BumpOffset_Accurate}.Result) {name=UVs, nodeColor=<rgba 0.875 1 0 1>}
  NamedRerouteDeclaration(None=Power(Base=@n8.EmissiveColor, Exp=ScalarParameter() {defaultValue=1, group=Visor, parameterName=Visor_EmissionPower, sortPriority=43}) {constExponent=1}) {name=Emissive_IN, nodeColor=<rgba 1 0 0.9609 1>}
  n18 = NamedRerouteUsage()
  n19 = ScalarParameter() {defaultValue=1, group=Visor, parameterName=Visor_BaseBrightness, sortPriority=41}
  n20 = ScalarParameter() {defaultValue=80, group=Visor Dots, parameterName=Dots_NumberOfDots, sortPriority=65}
  n21 = Constant() {r=1}
  n22 = Constant() {r=0.75}
  n23 = Constant() {r=0.5}
  n24 = Constant() {r=-0.25}
  n25 = Constant() {r=1}
  n26 = TextureCoordinate()
  n27 = NamedRerouteUsage()
  n28 = VectorParameter() {group=Visor Dots, parameterName=Dots_UVOffset, sortPriority=65, defaultValue=<rgba 0.7192 0.8 0 0>}
  n29 = ComponentMask(None=@n28.RGB) {g=true, r=true}
  n30 = ScalarParameter() {defaultValue=0.0672, group=Visor Dots, parameterName=Dots_MaskRadius, sortPriority=65}
  n31 = ScalarParameter() {defaultValue=1.1476, group=Visor Dots, parameterName=Dots_MaskDensity, sortPriority=65}
  n32 = ComponentMask(None=VectorParameter() {group=Visor Dots, parameterName=Dots_UVOffset2, sortPriority=65, defaultValue=<rgba 0.56 0.8 0 0>}.RGB) {g=true, r=true}
  n33 = Constant2Vector() {g=-0.03}
  n34 = Constant() {r=0.08}
  n35 = Constant() {r=4}
  n36 = TextureCoordinate()
  n37 = Constant2Vector() {g=-0.07}
  n38 = Add(A=@n29, B=@n37)
  n39 = Constant() {r=0.06}
  n40 = Multiply(
   A=MaterialFunctionCall(UVs=TextureCoordinate(), CenterPosition=Add(A=@n29, B=@n33), Radius=@n34, Density=@n35) {materialFunction=&RadialGradientExponential}.RadialGradientExponential
   B=OneMinus(None=MaterialFunctionCall(UVs=@n36, CenterPosition=@n38, Radius=@n39, Density=@n35) {materialFunction=&RadialGradientExponential}.RadialGradientExponential)
  )
  n41 = TextureCoordinate()
  n42 = Add(A=@n32, B=@n37)
  n43 = Multiply(
   A=MaterialFunctionCall(UVs=TextureCoordinate(), CenterPosition=Add(A=@n32, B=@n33), Radius=@n34, Density=@n35) {materialFunction=&RadialGradientExponential}.RadialGradientExponential
   B=OneMinus(None=MaterialFunctionCall(UVs=@n41, CenterPosition=@n42, Radius=@n39, Density=@n35) {materialFunction=&RadialGradientExponential}.RadialGradientExponential)
  )
  n44 = LinearInterpolate(B=Add(A=@n40, B=@n43), Alpha=NamedRerouteUsage()) {constA=1}
  n45 = Step(Y=Subtract(A=ComponentMask(None=@n29) {r=true}) {constB=0.1}, X=ComponentMask(None=TextureCoordinate()) {r=true})
  n46 = Add(A=@n40, B=OneMinus(None=@n45)) {constB=0}
  n47 = NamedRerouteUsage()
  n48 = Multiply(
   A=Add(
    A=MaterialFunctionCall(UVs=@n26, CenterPosition=@n29, Radius=@n30, Density=@n31) {materialFunction=&RadialGradientExponential}.RadialGradientExponential
    B=MaterialFunctionCall(UVs=@n26, CenterPosition=@n32, Radius=@n30, Density=@n31) {materialFunction=&RadialGradientExponential}.RadialGradientExponential
   )
   B=LinearInterpolate(A=LinearInterpolate(A=@n44, B=@n46, Alpha=@n47), B=Add(A=@n45, B=@n43), Alpha=NamedRerouteUsage())
  )
  n49 = Multiply(
   A=Multiply(
    A=MaterialFunctionCall!(UV=@n26, NumDotsX=@n20, NumDotsY=@n20, SharpRadius=@n21, GlowDensity=@n22, GlowRadius=@n23, GlowMin=@n24, GlowMax=@n25) {materialFunction=/AudioWidgets/AudioMaterialSlate/Functions/MF_Dots}.Glow
    B=OneMinus(None=@n27)
   )
   B=@n48
  )
  n50 = VectorParameter() {group=Visor, parameterName=Visor_MaskTintColor, defaultValue=<rgba 1 1 1 1>}
  n51 = ScalarParameter() {defaultValue=1, group=Visor Dots, parameterName=Dots_Intensity, sortPriority=65}
  n52 = ScalarParameter() {defaultValue=1, parameterName=EmissiveBrightness}
  n53 = Add(
   A=LinearInterpolate(
    A=VectorParameter() {group=Visor, parameterName=Visor_FresnelTintColor, sortPriority=43, defaultValue=<rgba 0 0 0 1>}.RGB
    B=Multiply(A=@n18, B=@n19)
    Alpha=NamedRerouteUsage()
   )
   B=Multiply(A=Multiply(A=@n49, B=@n50.RGB), B=Add(A=@n51, B=@n52)) {constB=3}
  )
  n54 = NamedRerouteUsage()
  NamedRerouteDeclaration(None=Add(A=Multiply(A=@n53, B=@n54), B=NamedRerouteUsage())) {name=Emissive_OUT, nodeColor=<rgba 0 1 0.7188 1>}
  NamedRerouteDeclaration() {name=BasicVisor, nodeColor=<rgba 0.8984 1 0 1>}
  n55 = Constant() {r=12}
  n56 = Divide(A=Time(), B=@n55) {constB=10}
  n57 = Constant() {r=0.6}
  n58 = Floor(None=Add(A=Frac(None=@n56), B=Divide(A=@n57, B=@n55) {constB=10}))
  n59 = Multiply(A=@n58) {constB=3}
  n60 = Constant() {r=36}
  n61 = Add(A=Divide(A=Time(), B=@n60) {constB=10}) {constB=0.5}
  n62 = Constant() {r=1.25}
  n63 = Floor(None=Add(A=Frac(None=@n61), B=Divide(A=@n62, B=@n60) {constB=10}))
  NamedRerouteDeclaration(None=LinearInterpolate(A=LinearInterpolate(A=@n59, B=@n63, Alpha=@n63), B=Multiply(A=@n14) {constB=2}, Alpha=@n14)) {name=FlipbookIndex, nodeColor=<rgba 0 1 0.1797 1>}
  n64 = Add(A=TextureCoordinate(), B=Constant2Vector() {g=-0.3, r=-0.14})
  n65 = Constant() {r=0.3}
  n66 = Constant2Vector() {g=1}
  n67 = NamedRerouteDeclaration(None=If(
    A=MaterialFunctionCall!(UVs=@n64, Radius=@n65) {materialFunction=/FortUILibrary/MaterialFunctions/Shapes/SDFs/MF_SDF_Circle}.SDF
    "A > B"=@n66.R
    "A < B"=@n66.R
   )) {name=VisorMask, nodeColor=<rgba 1 0 0.5156 1>}
  NamedRerouteDeclaration(None=@n67) {name=TVStaticMask, nodeColor=<rgba 0.7109 1 0 1>}
  NamedRerouteDeclaration(None=@n67) {name=ScanlinesMask, nodeColor=<rgba 1 0.5116 0.5981 1>}
  NamedRerouteDeclaration(None=@n67) {name=SweepDistortionMask, nodeColor=<rgba 0.161 1 0.707 1>}
  NamedRerouteDeclaration() {name=Debug, nodeColor=<rgba 0 1 0.5078 1>}
  NamedRerouteDeclaration(None=@n5) {name=EyeInnerArea, nodeColor=<rgba 0.125 0 1 1>}
  NamedRerouteDeclaration(None=@n63) {name=LeftWink, nodeColor=<rgba 0.0781 1 0 1>}
  NamedRerouteDeclaration(None=@n58) {name=Blink, nodeColor=<rgba 1 0 0.3984 1>}
  NamedRerouteUsage()
  NamedRerouteUsage()
  Saturate(None=@n48)
 ]}
