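# Animation Blueprint for the FutureRabbit backpack that periodically wiggles its two ear bones on top of the incoming pose.
# Two additive bone transforms apply pitch from two summed real-time sine waves, roughly mirrored between the ears with a slight offset.
# Delay-driven logic in the update event toggles a shared Alpha between moving and still, varying the interval between two durations.
# Alpha interpolation on the bone nodes smooths these transitions.
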AnimBlueprint /BRCosmetics/Accessories/FORT_Backpacks/Backpack_F_MED_FutureRabbit/Meshes/F_MED_FutureRabbit_Pack_AnimBP
target_skeleton=&/F_MED_FutureRabbit_Pack_Skeleton
graphs [
 AnimGraph {nodes [
   n1 = Get Alpha() -> (Alpha:double)
   n2 = "Transform (Modify) Bone - Bone: MSTR_ear_r"(
    ComponentPose=Local To Component(LocalPose=Input Pose())
    Alpha=@n1
    Translation="0,0,0"
    Rotation=Get ear_rotation_r() -> (ear_rotation_r:Rotator)
    Scale="1,1,1"
   ) {rotationMode=BMM_Additive, alphaScaleBiasClamp {bInterpResult=true, interpSpeedDecreasing=7}, boneToModify {boneName=MSTR_ear_r}}
   n3 = "Transform (Modify) Bone - Bone: MSTR_ear_l"(ComponentPose=@n2, Alpha=@n1, Translation="0,0,0", Rotation=Get ear_rotation_l() -> (ear_rotation_l:Rotator), Scale="1,1,1") {rotationMode=BMM_Additive, alphaScaleBiasClamp {bInterpResult=true, interpSpeedDecreasing=7}, boneToModify {boneName=MSTR_ear_l}}
   Output Pose(Result=Component To Local(ComponentPose=@n3)) {layerGroup=None}
  ]}
 EventGraph {nodes [
   Event BlueprintUpdateAnimation()
   > n1 = Sequence()
   > n2 = Branch(Condition=Get Condition() -> (Condition:bool))
   > Delay(Duration=Get Duration() -> (Duration:float))
   > Set Alpha(Alpha:double=0) -> (Output_Get:double)
   > Set Condition(Condition:bool=false) -> (Output_Get:bool)
   n3 = Sequence(execute=@n2.else)
   > n4 = Branch(Condition=Get First Pause() -> (First Pause:bool))
   > n5 = Set Duration(Duration:float=4) -> (Output_Get:float)
   > Delay(Duration=@n5)
   > Set First Pause(First Pause:bool=false) -> (Output_Get:bool)
   n6 = Set Duration(execute=@n4.else, Duration=2)
   > Delay(Duration=@n6)
   > Set First Pause(First Pause=true)
   Delay(execute=@n3.then_1, Duration=Get Duration())
   > Set Alpha(Alpha=1)
   > Set Condition(Condition=true)
   n7 = GetRealTimeSeconds()
   n8 = float * float(A=@n7, B=18) {is=title K2Node_PromotableOperator}
   n9 = float * float(A=@n7, B=16)
   n10 = float + float(A=float * float(A="Sin (Radians)"(A=@n9), B=4.5), B=float * float(A="Sin (Radians)"(A=@n8), B=2))
   n11 = Set Current(execute=@n1.then_1, Current:double=@n10) -> (Output_Get:double)
   > Set ear_rotation_r(ear_rotation_r:Rotator=MakeRotator(Pitch=@n11)) -> (Output_Get:Rotator)
   > Set ear_rotation_l(ear_rotation_l:Rotator=MakeRotator(Pitch=float - float(A=-1, B=@n10))) -> (Output_Get:Rotator)
   TryGetPawnOwner()
  ]}
]
variables [
 Alpha {type=number}
 Condition {type=boolean, default=true}
 Current {type=number}
 Duration {type=number, default=4}
 ear_rotation_l {type=Rotator}
 ear_rotation_r {type=Rotator, default=<rot 0 0 0.5042>}
 First Pause {type=boolean, default=true}
]
